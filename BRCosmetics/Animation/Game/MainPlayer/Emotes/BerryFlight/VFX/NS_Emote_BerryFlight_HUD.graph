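# Niagara system for the BerryFlight emote: a HUD of meshes and glow sprites at the character's head, with accompanying lights.
# The HUD eases toward the FX_Head socket, C_SpinLerp blends its orientation toward a fixed rotation, and glows narrow by camera-to-root alignment.
# Blue and red head lights plus two root-relative ground lights use front-end-specific radius and exponent.
# C_HudFlicker modulates HUD opacity and light intensity.
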NiagaraSystem /BRCosmetics/Animation/Game/MainPlayer/Emotes/BerryFlight/VFX/NS_Emote_BerryFlight_HUD
dependencies {
 usedDataInterfaces [&NiagaraDataInterfaceCurve, &NiagaraDataInterfaceSkeletalMesh, &NiagaraDataInterfaceCamera]
 usedDynamicInputs []
 usedModules [&SystemState]
}
emitters {
 BLUE_LIGHT {
  emitterUpdateScript [
   &EmitterState(Life Cycle Mode=::NewEnumerator1, Loop Behavior=::NewEnumerator1, Loop Duration=1, UseLoopDelay=false)
   &SpawnBurst_Instantaneous(Use Spawn Probability=false, UseLoopCountLimit=false)
  ]
  particleSpawnScript [&InitializeParticle(Color=<rgba 0 0.8775 1 0.05>, Sprite Size Mode=::NewEnumerator3, UsePositionOffset=false, Write Material Random=true)]
  particleUpdateScript [
   &ParticleState(Kill Particles When Lifetime Has Elapsed=false)
   &SkeletalMeshLocation(
    MaskSpawnGroup=false
    Mesh Sampling Type=::NewEnumerator5
    Orientation Sampling=::NewEnumerator1
    Sampled Position Offset=<0 30 0>
    Skeletal Mesh=DataInterface() {propertyValues {
      DefaultMesh=null
      ExcludeBoneName=root
      OptionalComponentClass=null
      PreviewMesh=null
      SkinningMode=SkinOnTheFly
      SoftSourceActor=null
      SourceMode=Default
      UvSetIndex=0
      WholeMeshLOD=-1
      bExcludeBone=false
      bReadDeformedGeometry=true
      bRequireCurrentFrameData=true
      ComponentTags []
      FilteredBones []
      FilteredSockets [FX_Head]
      MeshUserParameter {parameter {name=None, typeDefHandle {registeredTypeIndex=1}, varData []}}
      SamplingRegions []
     }}
    Socket Sampling Mode=::NewEnumerator5
   )
   &FN_Light_Attributes(
    Color=$Particles.Color
    Create Color Attribute=true
    Create Exponent Attribute=true
    Create Radius Attribute=true
    Exponent:NiagaraFloat=&FN_FrontEndFloatModifier(Gameplay=8, Front End=50)
    Radius=&FN_FrontEndFloatModifier(Gameplay=150, Front End=100)
    Scale Color=&FloatFromCurve(
     FloatCurve=DataInterface() {propertyValues {Curve {keys [<at 0 0>,<at 0.02 1>,<at 1 1>]}}}
     Scale Curve=&Lerp_Float(A=1, B=5, Alpha=$User.C_HudFlicker)
    )
   )
  ]
  rendererClasses [&NiagaraLightRendererProperties]
  renderers [{rendererClass=&NiagaraLightRendererProperties, rendererIndex=0}]
 }
 Ground_Light {
  emitterUpdateScript [
   &EmitterState(Life Cycle Mode=::NewEnumerator1, Loop Behavior=::NewEnumerator1, Loop Duration=1, UseLoopDelay=false)
   &SpawnBurst_Instantaneous(Spawn Count=2, Use Spawn Probability=false, UseLoopCountLimit=false)
  ]
  particleSpawnScript [
   &InitializeParticle(Color=<rgba 0 0.3922 1 0.05>, Sprite Size Mode=::NewEnumerator3, UsePositionOffset=false, Write Material Random=true)
   &PartitionParticles()
  ]
  particleUpdateScript [
   &ParticleState(Kill Particles When Lifetime Has Elapsed=false)
   &SkeletalMeshLocation(
    MaskSpawnGroup=false
    Mesh Sampling Type=::NewEnumerator5
    Orientation Sampling=::NewEnumerator1
    Sampled Position Offset=&MergeHelper_VectorSwitch(0=<50 70 15>, 1=<-50 70 15>, 2=?, 3=?, Index=$Particles.PartitionParticles.Partition, Num=::NewEnumerator0)
    Sampled Position Offset Coordinate Space:ENiagaraSkelMeshTransforms=::NewEnumerator2
    Skeletal Mesh=DataInterface() {propertyValues {
      DefaultMesh=null
      ExcludeBoneName=root
      OptionalComponentClass=null
      PreviewMesh=null
      SkinningMode=SkinOnTheFly
      SoftSourceActor=null
      SourceMode=Default
      UvSetIndex=0
      WholeMeshLOD=-1
      bExcludeBone=false
      bReadDeformedGeometry=true
      bRequireCurrentFrameData=true
      ComponentTags []
      FilteredBones []
      FilteredSockets [root]
      MeshUserParameter {parameter {name=None, typeDefHandle {registeredTypeIndex=1}, varData []}}
      SamplingRegions []
     }}
    Socket Sampling Mode=::NewEnumerator5
   )
   &FN_Light_Attributes(
    Color=$Particles.Color
    Create Color Attribute=true
    Create Exponent Attribute=true
    Create Radius Attribute=true
    Exponent:NiagaraFloat=&FN_FrontEndFloatModifier(Gameplay=4, Front End=50)
    Radius=&FN_FrontEndFloatModifier(Gameplay=60, Front End=100)
    Scale Color=&FloatFromCurve(
     FloatCurve=DataInterface() {propertyValues {Curve {keys [<at 0 0>,<at 0.02 1>,<at 1 1>]}}}
     Scale Curve=&Lerp_Float(
      A=1
      B=&FNTimeOfDayScalar(Night Scale=3.2, Morning Scale=3.2, Day Scale=3.2, Evening Scale=3.2)
      Alpha=$User.C_HudFlicker
     )
    )
   )
  ]
  rendererClasses [&NiagaraLightRendererProperties]
  renderers [{rendererClass=&NiagaraLightRendererProperties, rendererIndex=0}]
 }
 HUD {
  emitterUpdateScript [
   &EmitterState(Life Cycle Mode=::NewEnumerator1, Loop Behavior=::NewEnumerator1, Loop Duration=1, UseLoopDelay=false)
   &SpawnBurst_Instantaneous(Use Spawn Probability=false, UseLoopCountLimit=false)
  ]
  particleSpawnScript [
   &InitializeParticle(
    Color=<rgba 0.95 0.95 0.95 1.5>
    Lifetime Max=1.75
    Lifetime Min=1.4
    Lifetime Mode=::NewEnumerator1
    Mass Mode=::NewEnumerator3
    Mesh Scale Mode=::NewEnumerator3
    Mesh Uniform Scale=2.5
    Sprite Rotation Mode=::NewEnumerator0
    Sprite Size Mode=::NewEnumerator3
    Uniform Sprite Size=250
    UsePositionOffset=false
    Write Material Random=true
   )
   &InitialMeshOrientation(Mesh Axis To Orient=<0 0 0>, Mesh Orientation Mode=::NewEnumerator2, Orientation Vector=<0 -0.5 0>, Use Rotation Vector=true)
   /SetVariables(Particles.BlueGlowSize:Vector2f=<500 500>, Particles.RedGlowSize:Vector2f=<200 200>) {isSetParameters=true}
   &CameraOffset()
  ]
  particleUpdateScript [
   &ParticleState(Kill Particles When Lifetime Has Elapsed=false)
   &ScaleColor(
    Scale Alpha=&FloatFromCurve(FloatCurve=DataInterface() {propertyValues {Curve {
        defaultValue=0
        keys [{interpMode=RCIM_Cubic, tangentMode=RCTM_User, time=0, value=0}, {interpMode=RCIM_Cubic, tangentMode=RCTM_User, time=0.4, value=1}]
       }}})
    ScaleA=true
    ScaleRGB=true
   )
   &ScaleColor(Scale Alpha=&Lerp_Float(A=0.3, B=1, Alpha=$User.C_HudFlicker), ScaleA=true, ScaleRGB=true) {module=ScaleColor001}
   &SolveForcesAndVelocity(Clamp Velocity=false, Limit Acceleration=false)
   &SkeletalMeshLocation(
    MaskSpawnGroup=false
    Mesh Sampling Type=::NewEnumerator5
    Orientation Sampling=::NewEnumerator1
    Position Sampling=::NewEnumerator2
    Sampled Position Offset=<0 2 0>
    Skeletal Mesh=DataInterface() {propertyValues {
      DefaultMesh=null
      ExcludeBoneName=root
      OptionalComponentClass=null
      PreviewMesh=null
      SkinningMode=SkinOnTheFly
      SoftSourceActor=null
      SourceMode=Default
      UvSetIndex=0
      WholeMeshLOD=-1
      bExcludeBone=false
      bReadDeformedGeometry=true
      bRequireCurrentFrameData=true
      ComponentTags []
      FilteredBones []
      FilteredSockets [FX_Head]
      MeshUserParameter {parameter {name=None, typeDefHandle {registeredTypeIndex=1}, varData []}}
      SamplingRegions []
     }}
    Socket Sampling Mode=::NewEnumerator5
   )
   &SkeletalMeshLocation(
    Bone Sampling Mode=::NewEnumerator4
    MaskSpawnGroup=false
    Orientation Sampling=::NewEnumerator1
    Position Sampling=::NewEnumerator2
    Skeletal Mesh=DataInterface() {propertyValues {
      DefaultMesh=null
      ExcludeBoneName=root
      OptionalComponentClass=null
      PreviewMesh=null
      SkinningMode=SkinOnTheFly
      SoftSourceActor=null
      SourceMode=Default
      UvSetIndex=0
      WholeMeshLOD=-1
      bExcludeBone=false
      bReadDeformedGeometry=true
      bRequireCurrentFrameData=true
      ComponentTags []
      FilteredBones [root]
      FilteredSockets [None]
      MeshUserParameter {parameter {name=None, typeDefHandle {registeredTypeIndex=1}, varData []}}
      SamplingRegions []
     }}
   ) {module=SkeletalMeshLocation001}
   /SetVariables(Particles.Position=&self:ScratchDynamicInput_01(
     RateOfChange:NiagaraFloat=&FloatFromCurve(FloatCurve=DataInterface() {propertyValues {Curve {defaultValue=1, keys [{tangentMode=RCTM_User, time=0, value=5}, {tangentMode=RCTM_User, time=1, value=0.5}]}}})
     InitialPosition:NiagaraPosition=$Particles.Position
     TargetPosition:NiagaraPosition=$Particles.SkeletalMeshLocation.SampledPosition
    )) {isSetParameters=true}
   /SetVariables(Particles.BlueGlowPos:NiagaraPosition=&self:ScratchDynamicInput_01(
     RateOfChange:NiagaraFloat=&FloatFromCurve(FloatCurve=DataInterface() {propertyValues {Curve {defaultValue=1, keys [{tangentMode=RCTM_User, time=0, value=5}, {tangentMode=RCTM_User, time=1, value=0.5}]}}})
     InitialPosition:NiagaraPosition=&AddVectorToPosition(Position=$Particles.Position, Vector=<0 50 0>)
     TargetPosition:NiagaraPosition=&AddVectorToPosition(Position=$Particles.SkeletalMeshLocation.SampledPosition, Vector=<0 50 0>)
    )) {isSetParameters=true}
   /SetVariables(Particles.ReticlePos:NiagaraPosition=&self:ScratchDynamicInput_01(
     RateOfChange:NiagaraFloat=&FloatFromCurve(FloatCurve=DataInterface() {propertyValues {Curve {defaultValue=1, keys [{tangentMode=RCTM_User, time=0, value=5}, {tangentMode=RCTM_User, time=1, value=0.5}]}}})
     InitialPosition:NiagaraPosition=$Particles.Position
     TargetPosition:NiagaraPosition=$Output.SkeletalMeshLocation.SampledPositionWithoutOffset
    )) {isSetParameters=true}
   /SetVariables(Particles.RedGlowPos:NiagaraPosition=&AddVectorToPosition(
     Position=&self:ScratchDynamicInput_01(
      RateOfChange:NiagaraFloat=&FloatFromCurve(FloatCurve=DataInterface() {propertyValues {Curve {defaultValue=1, keys [{tangentMode=RCTM_User, time=0, value=5}, {tangentMode=RCTM_User, time=1, value=0.5}]}}})
      InitialPosition:NiagaraPosition=$Particles.Position
      TargetPosition:NiagaraPosition=$Output.SkeletalMeshLocation.SampledPositionWithoutOffset
     )
     Vector=&Lerp_Vector(
      A=<0 65 0>
      B=<0 72.5 0>
      Alpha=&Abs_Float(Float:NiagaraFloat=&DotProduct(
        Vector A=&Multiply_Vector(
         A=&ReturnMeshOrientationAxis(Mesh Orientation Axis=::XAxis, Mesh Orientation Quat=$Particles.SkeletalMeshLocation001.SampledRotation)
         B=&ReturnCameraVector(Camera Vectors=::NewEnumerator1, Coordinate Space=::Simulation)
        )
        Vector B=&ReturnMeshOrientationAxis(Mesh Orientation Axis=::XAxis, Mesh Orientation Quat=$Particles.SkeletalMeshLocation001.SampledRotation)
       ))
     )
    )) {isSetParameters=true}
   /SetVariables(Particles.MeshOrientation=&self:ScratchDynamicInput(
     Change:NiagaraFloat=2
     Initial:Quat4f=$Particles.Initial.InitialRotation
     Target:Quat4f=&LerpQuaternion(
      Quaternion A:Quat4f=&MultiplyQuaternion(
       Quaternion A=$Particles.SkeletalMeshLocation.SampledRotation
       Quaternion B=&MultiplyQuaternion(
        Quaternion A=&MakeQuaternion(
         Quaternion From=0
         Axis=<0 1 0>
         Angle=90
         From=?
         To=?
         Stabilization Mode:ENiagara_OrientVectorCount=::NewEnumerator0
         Side Direction:Vector3f=?
         Up Direction:Vector3f=?
         Angle Type=::NewEnumerator0
         Normalize:NiagaraBool=true
        )
        Quaternion B=&MakeQuaternion(
         Quaternion From=0
         Axis=<0 0 1>
         Angle=-90
         From=?
         To=?
         Stabilization Mode:ENiagara_OrientVectorCount=::NewEnumerator0
         Side Direction:Vector3f=?
         Up Direction:Vector3f=?
         Angle Type=::NewEnumerator0
         Normalize:NiagaraBool=true
        )
       )
      )
      Quaternion B:Quat4f=&MultiplyQuaternion(
       Quaternion A=&MakeQuaternion(
        Quaternion From=0
        Axis=<0 0 1>
        Angle=-90
        From=?
        To=?
        Stabilization Mode:ENiagara_OrientVectorCount=::NewEnumerator0
        Side Direction:Vector3f=?
        Up Direction:Vector3f=?
        Angle Type=::NewEnumerator0
        Normalize:NiagaraBool=true
       )
       Quaternion B=&MakeQuaternion(
        Quaternion From=0
        Axis=<1 0 0>
        Angle=30
        From=?
        To=?
        Stabilization Mode:ENiagara_OrientVectorCount=::NewEnumerator0
        Side Direction:Vector3f=?
        Up Direction:Vector3f=?
        Angle Type=::NewEnumerator0
        Normalize:NiagaraBool=true
       )
      )
      Lerp Factor:NiagaraFloat=$User.C_SpinLerp
      Enable Spherical Linear Interpolation:NiagaraBool=false
     )
    )) {isSetParameters=true}
   /SetVariables(Particles.ReticleOrientation:Quat4f=&self:ScratchDynamicInput(
     Change:NiagaraFloat=&Lerp_Float(A=6, B=6, Alpha=$User.C_SpinLerp)
     Initial:Quat4f=$Particles.Initial.InitialRotation
     Target:Quat4f=&MultiplyQuaternion(
      Quaternion A=$Particles.SkeletalMeshLocation.SampledRotation
      Quaternion B=&MultiplyQuaternion(
       Quaternion A=&MakeQuaternion(
        Quaternion From=0
        Axis=<0 1 0>
        Angle=90
        From=?
        To=?
        Stabilization Mode:ENiagara_OrientVectorCount=::NewEnumerator0
        Side Direction:Vector3f=?
        Up Direction:Vector3f=?
        Angle Type=::NewEnumerator0
        Normalize:NiagaraBool=true
       )
       Quaternion B=&MakeQuaternion(
        Quaternion From=0
        Axis=<0 0 1>
        Angle=-90
        From=?
        To=?
        Stabilization Mode:ENiagara_OrientVectorCount=::NewEnumerator0
        Side Direction:Vector3f=?
        Up Direction:Vector3f=?
        Angle Type=::NewEnumerator0
        Normalize:NiagaraBool=true
       )
      )
     )
    )) {isSetParameters=true}
   /SetVariables(
    Particles.GlowCol:LinearColor=&MultiplyLinearColor(
     Linear Color=<rgba 0.1176 0.4329 1 0.115>
     Scale Factor=&MakeVector4(
      X=1
      Y=1
      Z=1
      W=&Multiply_Float(
       A=&FloatFromCurve(FloatCurve=DataInterface() {propertyValues {Curve {
           defaultValue=0
           keys [{interpMode=RCIM_Cubic, tangentMode=RCTM_User, time=0, value=0}, {interpMode=RCIM_Cubic, tangentMode=RCTM_User, time=0.4, value=1}]
          }}})
       B=&Lerp_Float(A=0.8, B=1, Alpha=$User.C_HudFlicker)
      )
     )
    )
    Particles.RedCol:LinearColor=&MultiplyLinearColor(
     Linear Color=<rgba 1 0 0.121 0.1>
     Scale Factor=&MakeVector4(
      X=1
      Y=1
      Z=1
      W=&Multiply_Float(
       A=&FloatFromCurve(FloatCurve=DataInterface() {propertyValues {Curve {
           defaultValue=0
           keys [{interpMode=RCIM_Cubic, tangentMode=RCTM_User, time=0, value=0}, {interpMode=RCIM_Cubic, tangentMode=RCTM_User, time=0.4, value=1}]
          }}})
       B=&Lerp_Float(A=0.8, B=1, Alpha=$User.C_HudFlicker)
      )
     )
    )
   ) {isSetParameters=true}
   /SetVariables(
    Particles.XFactor:NiagaraFloat=&Abs_Float(Float:NiagaraFloat=&DotProduct(
      Vector A=&Multiply_Vector(
       A=&ReturnMeshOrientationAxis(Mesh Orientation Axis=::XAxis, Mesh Orientation Quat=$Particles.SkeletalMeshLocation001.SampledRotation)
       B=&ReturnCameraVector(Camera Vectors=::NewEnumerator1, Coordinate Space=::Simulation)
      )
      Vector B=&ReturnMeshOrientationAxis(Mesh Orientation Axis=::XAxis, Mesh Orientation Quat=$Particles.SkeletalMeshLocation001.SampledRotation)
     ))
    Particles.ZFactor:NiagaraFloat=&Abs_Float(Float:NiagaraFloat=&DotProduct(
      Vector A=&Multiply_Vector(
       A=&ReturnMeshOrientationAxis(Mesh Orientation Axis=::ZAxis, Mesh Orientation Quat=$Particles.SkeletalMeshLocation001.SampledRotation)
       B=&ReturnCameraVector(Camera Vectors=::NewEnumerator1, Coordinate Space=::Simulation)
      )
      Vector B=&ReturnMeshOrientationAxis(Mesh Orientation Axis=::ZAxis, Mesh Orientation Quat=$Particles.SkeletalMeshLocation001.SampledRotation)
     ))
   ) {isSetParameters=true}
   /SetVariables(
    Particles.BlueGlowSize:Vector2f=&Multiply_Vector2D(
     A=$Particles.Initial.BlueGlowSize
     B=&MakeVector2D(
      X=&Lerp_Float(A=1, B=0.4, Alpha=$Particles.XFactor)
      Y=&Lerp_Float(A=1, B=1, Alpha=$Particles.ZFactor)
     )
    )
    Particles.RedGlowSize:Vector2f=&Multiply_Vector2D(
     A=$Particles.Initial.RedGlowSize
     B=&MakeVector2D(
      X=&Lerp_Float(A=1, B=0.5, Alpha=$Particles.XFactor)
      Y=&Lerp_Float(A=1, B=1, Alpha=$Particles.ZFactor)
     )
    )
   ) {isSetParameters=true}
  ]
  rendererClasses [&NiagaraMeshRendererProperties, &NiagaraSpriteRendererProperties]
  renderers [
   {rendererClass=&NiagaraMeshRendererProperties, rendererIndex=0}
   {rendererClass=&NiagaraMeshRendererProperties, rendererIndex=1}
   {rendererClass=&NiagaraSpriteRendererProperties, rendererIndex=2}
   {rendererClass=&NiagaraSpriteRendererProperties, rendererIndex=3}
  ]
 }
 RED_LIGHT {
  emitterUpdateScript [
   &EmitterState(Life Cycle Mode=::NewEnumerator1, Loop Behavior=::NewEnumerator1, Loop Duration=1, UseLoopDelay=false)
   &SpawnBurst_Instantaneous(Use Spawn Probability=false, UseLoopCountLimit=false)
  ]
  particleSpawnScript [&InitializeParticle(Color=<rgba 0.1 0 0.0041 0.05>, Sprite Size Mode=::NewEnumerator3, UsePositionOffset=false, Write Material Random=true)]
  particleUpdateScript [
   &ParticleState(Kill Particles When Lifetime Has Elapsed=false)
   &SkeletalMeshLocation(
    MaskSpawnGroup=false
    Mesh Sampling Type=::NewEnumerator5
    Orientation Sampling=::NewEnumerator1
    Sampled Position Offset=<0 20 0>
    Skeletal Mesh=DataInterface() {propertyValues {
      DefaultMesh=null
      ExcludeBoneName=root
      OptionalComponentClass=null
      PreviewMesh=null
      SkinningMode=SkinOnTheFly
      SoftSourceActor=null
      SourceMode=Default
      UvSetIndex=0
      WholeMeshLOD=-1
      bExcludeBone=false
      bReadDeformedGeometry=true
      bRequireCurrentFrameData=true
      ComponentTags []
      FilteredBones []
      FilteredSockets [FX_Head]
      MeshUserParameter {parameter {name=None, typeDefHandle {registeredTypeIndex=1}, varData []}}
      SamplingRegions []
     }}
    Socket Sampling Mode=::NewEnumerator5
   )
   &FN_Light_Attributes(
    Color=$Particles.Color
    Create Color Attribute=true
    Create Exponent Attribute=true
    Create Radius Attribute=true
    Exponent:NiagaraFloat=&FN_FrontEndFloatModifier(Gameplay=1, Front End=50)
    Radius=&FN_FrontEndFloatModifier(Gameplay=30, Front End=100)
    Scale Color=&FloatFromCurve(
     FloatCurve=DataInterface() {propertyValues {Curve {keys [<at 0 0>,<at 0.02 1>,<at 1 1>]}}}
     Scale Curve=&Lerp_Float(A=50, B=150, Alpha=$User.C_HudFlicker)
    )
   )
  ]
  rendererClasses [&NiagaraLightRendererProperties]
  renderers [{rendererClass=&NiagaraLightRendererProperties, rendererIndex=0}]
 }
}
summary {
 systemName=NS_Emote_BerryFlight_HUD
 emitters [
  {emitterName=HUD, rendererClasses [&NiagaraMeshRendererProperties, &NiagaraSpriteRendererProperties]}
  {emitterName=BLUE_LIGHT, rendererClasses [&NiagaraLightRendererProperties]}
  {emitterName=RED_LIGHT, rendererClasses [&NiagaraLightRendererProperties]}
  {emitterName=Ground_Light, rendererClasses [&NiagaraLightRendererProperties]}
 ]
 userVariables [
  User.bIsFrontEnd {type=NiagaraBool}
  User.bIsFrontEndPreview {type=NiagaraBool}
  User.C_Flicker {type=NiagaraFloat, defaultValue=0, description=""}
  User.C_HudFlicker {type=NiagaraFloat, defaultValue=1, description=""}
  User.C_SpinLerp {type=NiagaraFloat, defaultValue=0, description=""}
  User.C_SpinLerp2 {type=NiagaraFloat, defaultValue=0, description=""}
 ]
}
